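# Translucent lit Niagara sprite material for waterfall mist, using volumetric per-vertex non-directional lighting and responsive AA.
# An 8x8 SubUV smoke texture advances with particle relative time, and its red channel is remapped by add, multiply and power parameters.
# That mask is softened by depth fade, particle alpha and camera depth fade.
# A rainbow function adds to particle color for emissive adjustments, while a Color parameter tints normalized particle color.
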Material /BlastBerryMapContent/Game/Environments/Water/M_Waterfall_Smoke3
graph {
 properties {
  bEnableResponsiveAA=true
  bUsedWithNiagaraSprites=true
  blendMode=BLEND_Translucent
  pixelDepthOffsetMode=PDOM_Legacy
  translucencyLightingMode=TLM_VolumetricPerVertexNonDirectional
  usedShadingModels=Default Lit
 }
 nodes [
  n1 = ParticleColor()
  MaterialFunctionCall(Emissive=Add(A=MaterialFunctionCall() {materialFunction=/Game/Tools/WaterfallGenerator/Art/MF_Rainbow}.Result, B=@n1.RGB)) {materialFunction=&MF_FX_EmissiveAdjustments}
  n2 = Constant() {r=8}
  n3 = MaterialFunctionCall!(
   U_Images=@n2
   V_Images=@n2
   SubUV Texture RGB=TextureObject() {samplerType=SAMPLERTYPE_LinearColor, texture=/Game/Tools/WaterfallGenerator/Art/T_Waterfall_SmokeShape2}
   Time Speed
   Manual Time Control=StaticBool() {value=true}
   Time=Multiply(A=ParticleRelativeTime()) {constB=8}
   UVs
  ) {materialFunction=/Game/Effects/Fort_Effects/Materials/_Functions/MF_SubUV}
  n4 = Multiply(
   A=Add(
    A=MaterialFunctionCall(Float3=@n3.RGB) {materialFunction=&BreakOutFloat3Components}.R
    B=ScalarParameter() {group=MistA, parameterName=MistA_GrayAdjust_Add}
   )
   B=ScalarParameter() {defaultValue=1, group=MistA, parameterName=MistA_GrayAdjust_Mul}
  )
  n5 = Saturate(None=Power(Base=Saturate(None=@n4), Exp=ScalarParameter() {defaultValue=1.5, group=MistA, parameterName=MistA_GrayAdjust_Power}))
  Multiply(
   A=Multiply(A=@n1.A, B=DepthFade(Opacity=@n5) {fadeDistanceDefault=200})
   B=MaterialFunctionCall(Fade Length=ScalarParameter() {parameterName="Fade Length (S)"}, Fade Offset=ScalarParameter() {parameterName="Fade Offset (S)"}) {materialFunction=&CameraDepthFade}.Result
  )
  Multiply(A=VectorParameter() {parameterName=Color, defaultValue=<rgba 1 1 1 1>}.RGB, B=Normalize(VectorInput=@n1.RGB))
 ]
}
